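# Translucent unlit VFX material for the Cattus beam in the CattusDoggus event, colored through a purple curve-atlas row and particle color.
# Power-warped UVs pan noise and ripple textures in reversible time, shaping a sine profile across the beam.
# Particle alpha offsets intensity, while sphere masks and depth fade soften opacity.
# World position offset uses SplineThicken or noisy displacement, with cheaper fallbacks for mobile and, in places, low quality.
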Material /Figment_S10/Effects/Fort_Effects/Events/CattusDoggus/Cattus/M_Cattus_Beam
graph {
 properties {
  bUsedWithBeamTrails=true
  bUsedWithMeshParticles=true
  bUsedWithParticleSprites=true
  blendMode=BLEND_Translucent
  pixelDepthOffsetMode=PDOM_Legacy
  shadingModel=MSM_Unlit
  usedShadingModels=Unlit
 }
 nodes [
  n1 = StaticSwitchParameter(True=TextureCoordinate() {coordinateIndex=1}, False=TextureCoordinate()) {parameterName=TexCoord0or1}
  n2 = Sine(None=Multiply(A=MaterialFunctionCall(Float2=@n1) {materialFunction=&BreakOutFloat2Components}.G) {constB=0.5})
  n3 = MaterialFunctionCall(Float2=StaticSwitchParameter(True=TextureCoordinate() {coordinateIndex=1, uTiling=3}, False=TextureCoordinate() {uTiling=3}) {parameterName=TexCoord0or1}) {materialFunction=&BreakOutFloat2Components}
  n4 = AppendVector(A=Power(Base=@n3.R) {constExponent=0.6}, B=@n3.G)
  n5 = Time()
  n6 = StaticSwitchParameter(True=Multiply(A=@n5) {constB=-1}, False=@n5) {parameterName=ReverseTimeline}
  n7 = Panner(Coordinate=@n4, Time=@n6) {speedX=-3}
  n8 = TextureSampleParameter2D(UVs=Panner(Coordinate=@n4, Time=@n6) {speedX=-4}) {parameterName=TexB, texture=/Game/Effects/Fort_Effects/Textures/T_Smooth_tile}
  n9 = Divide(A=Multiply(
    A=TextureSampleParameter2D(UVs=@n7) {parameterName=TexA, texture=/Game/Effects/Fort_Effects/Gameplay/SmokeDustExplosions/Textures/T_NoisePack_01}.R
    B=Multiply(A=@n8.R) {constB=2}
   )) {constB=15}
  n10 = MaterialFunctionCall(Float2=StaticSwitchParameter(True=TextureCoordinate() {coordinateIndex=1, uTiling=2.5}, False=TextureCoordinate() {uTiling=2.5}) {parameterName=TexCoord0or1}) {materialFunction=&BreakOutFloat2Components}
  n11 = Power(Base=@n10.R) {constExponent=0.8}
  n12 = TextureSampleParameter2D(UVs=Panner(Coordinate=AppendVector(A=@n11, B=@n10.G), Time=@n6) {speedX=-0.5}) {parameterName=TexC, samplerType=SAMPLERTYPE_Normal, texture=/Game/Effects/Fort_Effects/Gameplay/Water/Textures/T_WaterRipples_01}
  n13 = MaterialFunctionCall(Float2=StaticSwitchParameter(True=TextureCoordinate() {coordinateIndex=1, uTiling=2}, False=TextureCoordinate() {uTiling=2}) {parameterName=TexCoord0or1}) {materialFunction=&BreakOutFloat2Components}
  n14 = Power(Base=@n13.R) {constExponent=0.65}
  n15 = TextureSampleParameter2D(UVs=Panner(Coordinate=AppendVector(A=@n14, B=@n13.G), Time=@n6) {speedX=-0.8}) {parameterName=TexD, samplerType=SAMPLERTYPE_Normal, texture=&T_WaterRipples_01}
  n16 = MaterialFunctionCall(Float2=StaticSwitchParameter(True=TextureCoordinate() {coordinateIndex=1, uTiling=1.5}, False=TextureCoordinate() {uTiling=1.5}) {parameterName=TexCoord0or1}) {materialFunction=&BreakOutFloat2Components}
  n17 = AppendVector(A=Power(Base=@n16.R) {constExponent=0.5}, B=@n16.G)
  n18 = Panner(Time=@n6) {speedX=0.35}
  n19 = Divide(A=ComponentMask(None=TextureSample(UVs=@n18) {samplerType=SAMPLERTYPE_Normal, texture=/Figment_S10/Effects/Fort_Effects/Textures/T_Weap_Distort_N_02}.RGB) {g=true, r=true}) {constB=8}
  n20 = TextureSampleParameter2D(
   UVs=Panner(Coordinate=Add(A=@n17, B=@n19), Time=@n6) {speedX=-1.5, speedY=0.5}
   Apply View MipBias=TextureObject() {samplerType=SAMPLERTYPE_Normal, texture=/Figment_S10/Effects/Fort_Effects/Textures/Smoke/T_SwirlFlow_02_N}
  ) {parameterName=TexE, samplerType=SAMPLERTYPE_Normal, texture=&T_WaterRipples_01}
  n21 = LinearInterpolate(Alpha=MaterialFunctionCall(A=@n12.R, B=@n15.R, C=@n20.R, Alpha=@n2) {materialFunction=&3ColorBlend}.Result) {constA=2.2, constB=5}
  n22 = Constant() {r=2}
  n23 = Power(
   Base=Add(A=Power(Base=@n2), B=LinearInterpolate(Alpha=@n9) {constA=-0.02, constB=0.6})
   Exp=QualitySwitch(Default=FeatureLevelSwitch(Default=@n21, ES2=@n22, ES3_1=@n22, SM4_REMOVED=@n21), Low=@n22)
  )
  n24 = ParticleColor()
  n25 = Subtract(A=@n23, B=StaticSwitchParameter(True=OneMinus(None=@n24.A), False=@n24.A) {parameterName=Debug}) {constB=0}
  n26 = MaterialFunctionCall(Float2=StaticSwitchParameter(True=TextureCoordinate() {coordinateIndex=1}, False=TextureCoordinate()) {parameterName=TexCoord0or1}) {materialFunction=&BreakOutFloat2Components}
  n27 = Multiply(A=@n26.R, B=LinearInterpolate(Alpha=ParticleColor().A) {constA=1, constB=2}) {constB=0.5}
  n28 = TextureSample(UVs=Panner(Coordinate=AppendVector(A=@n27, B=@n26.G), Time=@n6) {speedX=-2.5, speedY=1}) {texture=/Figment_S10/Effects/Fort_Effects/Textures/Tile/T_BlasterBeamDistort}
  n29 = TextureCoordinate() {coordinateIndex=1, uTiling=2.8, vTiling=1.4}
  n30 = TextureCoordinate() {uTiling=2.8, vTiling=1.4}
  n31 = TextureSample(UVs=Panner(Coordinate=StaticSwitchParameter(True=@n29, False=@n30) {parameterName=TexCoord0or1}, Time=@n6) {speedX=-2, speedY=0.5}) {samplerType=SAMPLERTYPE_LinearColor, texture=/Game/Effects/Fort_Effects/Effects/Environment/Desert/Dust/Organize/T_Turbulence_Seamless}
  n32 = LinearInterpolate(Alpha=ComponentMask(None=MaterialFunctionCall!(Blend=@n28.R, Base=@n31.B) {materialFunction=/Engine/Functions/Engine_MaterialFunctions03/Blends/Blend_Lighten}.Result) {r=true}) {constA=0.77, constB=1.25}
  n33 = Constant() {r=1}
  n34 = Multiply(A=QualitySwitch(Default=FeatureLevelSwitch(Default=@n32, ES2=@n33, ES3_1=@n33, SM4_REMOVED=@n32), Low=@n33), B=@n23)
  n35 = StaticSwitchParameter(True=Constant(), False=Divide(A=@n24.A)) {parameterName=Debug}
  n36 = CurveAtlasRowParameter(CurveTime=Saturate(None=Subtract(A=@n34, B=@n35))) {
   atlas=&AthenaCurveAtlas01
   curve=/Game/Effects/Fort_Effects/Events/CattusDoggus/Cattus/Curve_Cattus_Purple
   defaultValue=104
  }
  n37 = StaticSwitchParameter(True=TextureCoordinate() {coordinateIndex=1}, False=TextureCoordinate()) {parameterName=TexCoord0or1}
  n38 = Constant2Vector() {g=0.5, r=0.7}
  n39 = LinearInterpolate(A=ScalarParameter() {defaultValue=0.7, parameterName=Low}, B=ScalarParameter() {defaultValue=1, parameterName=High}, Alpha=@n20.R) {constA=0.7}
  n40 = Saturate(None=Multiply(
    A=SphereMask(A=@n37, B=@n38, Radius=@n39) {attenuationRadius=0.7, hardnessPercent=0}
    B=ScalarParameter() {defaultValue=10, parameterName=StartSoftness}
   ) {constB=50})
  n41 = Multiply(
   A=SphereMask(A=@n37, B=Constant2Vector() {g=0.5, r=0.3}) {attenuationRadius=0.7, hardnessPercent=0}
   B=ScalarParameter() {defaultValue=8, parameterName=EndSoftness}
  ) {constB=8}
  n42 = Saturate(None=Multiply(A=@n41, B=@n41))
  n43 = Multiply(A=Saturate(None=@n23), B=Multiply(A=Multiply(A=@n25, B=@n36.A), B=Multiply(A=@n40, B=@n42)))
  n44 = DepthFade(Opacity=@n43, FadeDistance=ScalarParameter() {defaultValue=600, parameterName=DepthFade}) {fadeDistanceDefault=600}
  FeatureLevelSwitch(Default=@n44, ES2=@n43, ES3_1=@n43, SM4_REMOVED=@n44)
  n45 = Multiply(A=@n36, B=@n24.RGB)
  MaterialFunctionCall(LDR Input=Saturate(None=@n45), HDR Tint=@n45) {materialFunction=&DeriveHDRfromLDR}
  Multiply(
   A=@n24.A
   B=MaterialFunctionCall(Float2=TextureCoordinate()) {materialFunction=&BreakOutFloat2Components}.G
  )
  n46 = DynamicParameter() {paramNames [PanAnim, SplineWidthBase, SplineWidthTop, Erode]}
  n47 = Constant()
  n48 = TextureSample(UVs=Panner(Coordinate=TextureCoordinate() {uTiling=1.5}) {speedX=-1}) {samplerType=SAMPLERTYPE_Normal, texture=&T_Weap_Distort_N_02}
  n49 = MaterialFunctionCall(Float2=TextureCoordinate()) {materialFunction=&BreakOutFloat2Components}
  n50 = Multiply(A=AppendVector(A=ComponentMask(None=@n48.RGB) {g=true, r=true}, B=@n15.G), B=Sine(None=Multiply(A=@n49.R) {constB=0.5}))
  n51 = ParticleColor()
  n52 = StaticSwitchParameter(
   True=@n47
   False=Multiply(
    A=Transform(None=@n50) {transformSourceType=TRANSFORMSOURCE_Local, transformType=TRANSFORM_Tangent}
    B=LinearInterpolate(Alpha=@n51.A) {constA=1800, constB=2500}
   ) {constB=1500}
  ) {parameterName=Debug}
  StaticSwitchParameter(
   True=MaterialFunctionCall(
    WidthBase=@n46.SplineWidthBase
    WidthTip=@n46.SplineWidthBase
    Expand U or V UV Channel=StaticBoolParameter() {defaultValue=true, parameterName=Expand U or V UV Channel}
   ) {materialFunction=&SplineThicken}.WorldPositionOffset
   False=FeatureLevelSwitch(Default=@n52, ES2=@n47, ES3_1=@n47, SM4_REMOVED=@n52)
  ) {parameterName=UseSpline}
 ]
}
