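# Masked lit material for Cattus death effect sprites, tiny squares that dissolve around an offset point from the object.
# Erode (by default a debug time sine) moves a sphere-mask front that drives ash-like falling offset and color fade.
# An expanding ring, two subtracted sphere masks, sizes sprites through the Sprite function and tints their HDR glow.
# Panning noise shapes each sprite's mask, with optional dithered AA and depth fade off mobile.
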Material /Figment_S10/Effects/Fort_Effects/Events/CattusDoggus/Cattus/M_Cattus_Death_SpriteTinySquare
graph {
 properties {blendMode=BLEND_Masked, pixelDepthOffsetMode=PDOM_Legacy, usedShadingModels=Default Lit}
 nodes [
  n1 = WorldPosition()
  n2 = VectorParameter() {parameterName=PointOffset, defaultValue=<rgba 7000 -5500 0 1>}
  n3 = Add(A=Add(A=ObjectPositionWS(), B=ComponentMask(None=@n2.RGB) {b=true, g=true, r=true})) {constB=0}
  n4 = OneMinus(None=SphereMask(A=@n1.XYZ, B=@n3) {attenuationRadius=25000, hardnessPercent=0})
  n5 = ScalarParameter() {defaultValue=-0.5, parameterName=ErodeRangeLow}
  n6 = ScalarParameter() {defaultValue=10.5, parameterName=ErodeRangeHight}
  n7 = StaticSwitchParameter(
   True=MaterialFunctionCall(Speed=ScalarParameter() {defaultValue=0.2, parameterName=DebugTimeSpeed}) {materialFunction=&DebugTimeSine}.Result
   False=ScalarParameter() {parameterName=Erode}
  ) {defaultValue=true, parameterName=DebugErode}
  n8 = Multiply(A=Subtract(A=@n4, B=LinearInterpolate(A=@n5, B=@n6, Alpha=@n7) {constB=2})) {constB=-1}
  n9 = ComponentMask(None=VertexNormalWS()) {b=true, g=true, r=true}
  n10 = Constant3Vector() {constant=<rgba -0.05 -0.05 -0.7 0>}
  n11 = Constant3Vector() {constant=<rgba 0.1 0.1 -1 0>}
  n12 = VertexColor()
  n13 = Multiply(
   A=Multiply(A=Abs(None=@n9), B=LinearInterpolate(A=@n10, B=@n11, Alpha=@n12.R)) {constB=-1}
   B=LinearInterpolate(B=ScalarParameter() {defaultValue=10, parameterName=WPO_Strength}, Alpha=ScalarParameter() {parameterName=AshFallSpeed}) {constA=-2, constB=100}
  ) {constB=5}
  n14 = ScalarParameter() {defaultValue=45000, parameterName=Radius2}
  n15 = ScalarParameter() {defaultValue=30000, parameterName=Radius}
  n16 = Subtract(
   A=SphereMask(A=@n1.XYZ, B=@n3, Radius=LinearInterpolate(B=@n14, Alpha=@n7)) {attenuationRadius=18000, hardnessPercent=0}
   B=SphereMask(A=@n1.XYZ, B=@n3, Radius=LinearInterpolate(B=@n15, Alpha=@n7)) {attenuationRadius=10000, hardnessPercent=0}
  )
  n17 = Multiply(
   A=Multiply(A=@n12.R, B=ScalarParameter() {defaultValue=15, parameterName=SpriteSquareSize})
   B=Multiply(A=@n16, B=ScalarParameter() {defaultValue=5, parameterName=SphereMaskHandleSizeScale}) {constB=5}
  )
  n18 = MaterialFunctionCall(Distance=ScalarParameter() {defaultValue=3000, parameterName=WPO_NearCameraFade}) {materialFunction=&Near_Camera_Fade}
  Add(
   A=Multiply(A=Saturate(None=@n8), B=Multiply(A=@n13))
   B=MaterialFunctionCall!(Center Location=WorldPosition().XYZ, 0-1 UVs, XY scale=LinearInterpolate(B=@n17, Alpha=@n18.Result), OverridePivot, "Normals (Optional)") {materialFunction=&Sprite}.World position offset
  )
  n19 = Panner(Coordinate=TextureCoordinate() {uTiling=0.11, vTiling=0.11}, Time=Add(A=Time(), B=Multiply(A=@n12.R) {constB=20})) {speedX=0.02, speedY=0.02}
  n20 = SphereMask(
   A=TextureCoordinate()
   B=Constant() {r=0.5}
   Radius=LinearInterpolate(
    A=ScalarParameter() {parameterName=NoiseAmountLow}
    B=ScalarParameter() {defaultValue=0.9, parameterName=NoiseAmountHigh}
    Alpha=TextureSample(UVs=@n19) {texture=/Game/Effects/Fort_Effects/Textures/waveH}.R
   ) {constB=0.9}
  ) {attenuationRadius=0.5, hardnessPercent=30}
  n21 = Constant2Vector()
  n22 = Constant() {r=9000}
  n23 = StaticSwitchParameter(
   True=LinearInterpolate(
    A=MaterialFunctionCall(Alpha Threshold=@n20, Offset=@n21, Offset2=@n21) {materialFunction=&DitherTemporalAA_Fortnite}.Result
    B=@n20
    Alpha=MaterialFunctionCall(Distance=@n22) {materialFunction=&Near_Camera_Fade}.Result
   )
   False=@n20
  ) {parameterName=UseDitherAA}
  FeatureLevelSwitch(Default=@n23, ES2=@n20, ES3_1=@n20, SM4_REMOVED=@n23)
  n24 = TextureSample(UVs=Panner(Coordinate=TextureCoordinate() {uTiling=0.7, vTiling=0.7}) {speedX=0.1, speedY=0.1}) {texture=&waveH}
  n25 = VertexColor()
  n26 = Multiply(A=Saturate(None=@n16), B=@n20)
  n27 = ScalarParameter() {defaultValue=4.5, parameterName=SphereColorLerp}
  n28 = Add(A=VertexColor().R) {constB=0.6}
  n29 = LinearInterpolate(
   A=LinearInterpolate(
    A=VectorParameter() {parameterName=ColorHigh, defaultValue=<rgba 1 0.6706 0.2488 1>}.RGB
    B=VectorParameter() {parameterName=ColorHighBias, defaultValue=<rgba 1 0.0824 0.1222 1>}.RGB
    Alpha=Power(Base=@n25.B)
   ) {constB=0.056}
   B=VectorParameter() {parameterName=ColorLowBias, defaultValue=<rgba 0.0625 0.0112 0 1>}.RGB
   Alpha=Multiply(A=Multiply(A=@n26, B=@n27) {constB=10}, B=Saturate(None=@n28))
  )
  n30 = VectorParameter() {parameterName=ColorLow, defaultValue=<rgba 0.0625 0.0112 0 1>}
  n31 = Multiply(A=@n4, B=ScalarParameter() {defaultValue=12, parameterName=ErodeSoftness}) {constB=5}
  n32 = LinearInterpolate(
   A=ScalarParameter() {defaultValue=-0.5, parameterName=ErodeRangeLow}
   B=ScalarParameter() {defaultValue=10.5, parameterName=ErodeRangeHight}
   Alpha=@n7
  ) {constB=2}
  n33 = Power(Base=Saturate(None=Subtract(A=@n31, B=@n32)))
  n34 = LinearInterpolate(A=@n29, B=@n30.RGB, Alpha=@n33) {constA=1, constB=0}
  n35 = ScalarParameter() {defaultValue=15, parameterName=Emissive}
  MaterialFunctionCall(
   LDR Input=Multiply(A=LinearInterpolate(Alpha=@n24.R) {constA=0.7}, B=Saturate(None=@n34))
   HDR Tint=Multiply(A=@n29, B=LinearInterpolate(A=@n35, B=@n30.RGB, Alpha=@n33) {constA=1, constB=0})
  ) {materialFunction=&DeriveHDRfromLDR}
  n36 = DepthFade(Opacity=@n20, FadeDistance=ScalarParameter() {defaultValue=100, parameterName=DepthFadeDIstance})
  Multiply(
   A=FeatureLevelSwitch(Default=@n36, ES2=@n20, ES3_1=@n20, SM4_REMOVED=@n36)
   B=ScalarParameter() {defaultValue=1, parameterName=OverallOpacity}
  )
  Saturate(None=Add(A=@n12.R) {constB=0.9})
  TextureSample(UVs=@n19) {samplerType=SAMPLERTYPE_Normal, texture=/Figment_S10/Characters/Enemies/Husk/Textures/CloudImageNormals}
 ]
}
